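; Hard-surface QC trigger actor for Fortnite that gathers player assets for export and labels itself with its title.
; Cylinder overlap resets the export list, adds character and weapon preview icons, clears animation and mesh arrays, records the character name and binds cosmetic-applied delegates.
; An uncalled helper finds animation dependencies through the Asset Registry by path substrings, but returns inside its loop after the first match.
; The console-command helper is never called here.
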(event EventBeginPlay)

(event Collision|EventActorBeginOverlap
  (params (OtherActor :type Actor%20Object%20Reference)))

(event EventTick
  (params (DeltaSeconds :type Float%20%28single-precision%29)))

(event OnComponentBeginOverlap(Cylinder)
  (params (OverlappedComponent :type Primitive%20Component%20Object%20Reference) (OtherActor :type Actor%20Object%20Reference) (OtherComp :type Primitive%20Component%20Object%20Reference) (OtherBodyIndex :type Integer) (bFromSweep :type Boolean) (SweepResult :type Hit%20Result%20Structure%20%28by%20ref%29))
  (CallFunction|InitializeVariables)
  (CallFunction|GetIconsfromPlayer)
  (CallFunction|PrepThenExecuteCommand (Variables|Default|GetConsoleCommand)))

(event OnComponentEndOverlap(StaticMesh)
  (params (OverlappedComponent :type Primitive%20Component%20Object%20Reference) (OtherActor :type Actor%20Object%20Reference) (OtherComp :type Primitive%20Component%20Object%20Reference) (OtherBodyIndex :type Integer)))

(event Custom|PrepThenExecuteCommand
  (params (Command :type String))
  ; PrepThenExecuteCommandCollapse
  (CallFunction|ClearIndexData)
  (bind _charactername (CallFunction|GetCharacterNamefromPlayer))
  (bind _output_get (Variables|Default|SetCharacterName _charactername))
  (bind _asfort_player_pawn (Utilities|Casting|CastToFortPlayerPawn (Game|GetPlayerPawn 0))
    (:then
      (bind _delegatehandlecontroller_out (Cosmetics|WhenCosmeticPartApplied _asfort_player_pawn (delegate Custom|OnCosmeticPartApplied) true))
      (Variables|Default|SetDelegate%20Handle%20Controller _delegatehandlecontroller_out)
      (bind _delegatehandlecontroller_out_2 (Cosmetics|WhenCosmeticsApplied _asfort_player_pawn (delegate Custom|OnCosmeticsApplied)))
      (Variables|Default|SetDelegate%20Handle%20Controller%20Cosmetic%20Applied _delegatehandlecontroller_out_2))
    (:CastFailed
      (Development|PrintString "Cast Failed"))))

(fn Add_AnimSequence_To_Be_Exported
  (params (Object :type Object%20Reference) (PathKeyDirectory :type String :default "/Animation/") (PathKeyFileName :type String))
  ;[1 1 1] Clear Paths To Return
  (Utilities|Array|Clear (LocalVariables|GetPathsToReturn))
  (bind _outdependencies (CallFunction|TraceandFilterDependencies Object PathKeyDirectory PathKeyFileName))
  ;[1 1 1] For Each Traced/Filtered Dependency Get The Corresponding Asset
  (for _array_element _outdependencies
    (bind _returnvalue (AssetRegistry|GetAssetByObjectPath (AssetRegistry|GetAssetRegistry) _array_element false))
    ;[1 1 1] Get The Asset Path. If The Path Does Not Have An FBX Suffix, Add The Suffix To The Path. Store The Path.
    (if (Utilities|String|Contains (.RightS (Utilities|String|Split (Utilities|GetObjectPathString (AssetData|GetAsset _returnvalue)) "/" "IgnoreCase" "FromEnd")) ".")
      (LocalVariables|SetPath (Utilities|String|Split (Utilities|GetObjectPathString (AssetData|GetAsset _returnvalue)) "." "IgnoreCase" "FromEnd"))
      (else
        (LocalVariables|SetPath (Utilities|String|Append (Utilities|GetObjectPathString (AssetData|GetAsset _returnvalue)) ".FBX"))))
    ;[1 1 1] Update The Array Of Paths To Return
    (Utilities|Array|Add (LocalVariables|GetPathsToReturn) (LocalVariables|GetPath))
    (Utilities|LoadAssetBlocking (Utilities|ResolveSoftReference (AssetData|GetAsset _returnvalue)))
    (CallFunction|AddAssettobeExported (AssetData|GetAsset _returnvalue))
    (return (LocalVariables|GetPathsToReturn) PathKeyFileName)))

(fn Add_Asset_To_Be_Exported
  (params (Object :type Object%20Reference))
  (Utilities|Array|Add (Variables|Default|GetAssetsToExport) (Utilities|GetObjectPathString Object)))

(fn ClearIndexData
  (Utilities|Array|Clear (Variables|Default|GetWeaponAnimSequencesEquip))
  (Utilities|Array|Clear (Variables|Default|GetCharacterAnimSequencesEquip))
  (Utilities|Array|Clear (Variables|Default|GetWeaponAnimSequencesReload))
  (Utilities|Array|Clear (Variables|Default|GetCharacterAnimSequencesReload))
  (Utilities|Array|Clear (Variables|Default|GetWeaponAnimSequencesFire))
  (Utilities|Array|Clear (Variables|Default|GetCharacterAnimSequencesFire))
  (Utilities|Array|Clear (Variables|Default|GetWeaponAnimSequencesFireDownSights))
  (Utilities|Array|Clear (Variables|Default|GetCharacterAnimSequencesFireDownSights))
  (Utilities|Array|Clear (Variables|Default|GetWeaponSkeletalMeshPaths))
  (Utilities|Array|Clear (Variables|Default|GetCharacterSkeletalMeshPaths)))

(fn executeCommand
  (params (Command :type String) (self2 :type Pawn%20Object%20Reference))
  (bind _asplayer_controller (Utilities|Casting|CastToPlayerController (Pawn|GetController self2)))
  (Development|ExecuteConsoleCommand Command _asplayer_controller))

(fn GetCharacterNameFromPlayer
  (bind _asfort_player_pawn (Utilities|Casting|CastToFortPlayerPawn (Game|GetPlayerPawn 0))
    (:then
      (Development|PrintString (Utilities|String|Append "Character: " (.RightS (Utilities|String|Split (Utilities|GetObjectPathString (CosmeticLoadout|GetItemDefinitionfromLoadoutwithActor _asfort_player_pawn "(TagName=\"LoadoutSlot.Character\")")) "." "IgnoreCase" "FromEnd"))))
      (return (.RightS (Utilities|String|Split (Utilities|GetObjectPathString (CosmeticLoadout|GetItemDefinitionfromLoadoutwithActor _asfort_player_pawn "(TagName=\"LoadoutSlot.Character\")")) "." "IgnoreCase" "FromEnd"))))
    (:CastFailed
      (Development|PrintString "Cast Failed"))))

(fn GetIconsFromPlayer
  (bind _asfort_player_pawn (Utilities|Casting|CastToFortPlayerPawn (Game|GetPlayerPawn 0)))
  ;[1 1 1] Add Asset To Be Exported: Small Preview Image For Character
  (Variables|Default|SetCharacterIconSmallPath (Utilities|String|Append (Utilities|String|Split (Utilities|GetObjectPathString (Utilities|ResolveSoftReference (Item|GetSmallPreviewImage (CosmeticLoadout|GetItemDefinitionfromLoadoutwithActor _asfort_player_pawn "(TagName=\"LoadoutSlot.Character\")")))) "." "IgnoreCase" "FromEnd") ".TGA"))
  (Development|PrintString (Utilities|String|Append "Adding Character Small Preview Image Path" " : " (Utilities|String|Append (Utilities|String|Split (Utilities|GetObjectPathString (Utilities|ResolveSoftReference (Item|GetSmallPreviewImage (CosmeticLoadout|GetItemDefinitionfromLoadoutwithActor _asfort_player_pawn "(TagName=\"LoadoutSlot.Character\")")))) "." "IgnoreCase" "FromEnd") ".TGA")))
  (Development|PrintString (Utilities|String|Append "Fort Player Pawn Character Small Preview Image" " : " (Utilities|GetObjectPathString (Utilities|ResolveSoftReference (Item|GetSmallPreviewImage (CosmeticLoadout|GetItemDefinitionfromLoadoutwithActor _asfort_player_pawn "(TagName=\"LoadoutSlot.Character\")"))))))
  (CallFunction|AddAssettobeExported (Utilities|ResolveSoftReference (Item|GetSmallPreviewImage (CosmeticLoadout|GetItemDefinitionfromLoadoutwithActor _asfort_player_pawn "(TagName=\"LoadoutSlot.Character\")"))))
  ;[1 1 1] Add Asset To Be Exported: Small Preview Image For Weapon
  (Variables|Default|SetWeaponIconSmallPath (Utilities|String|Append (Utilities|String|Split (Utilities|GetObjectPathString (ItemPreviewImage|GetorLoadSynchronously (Item|GetWeaponDataSmallPreviewImage (Weapon|GetCurrentWeapon _asfort_player_pawn)))) "." "IgnoreCase" "FromEnd") ".TGA"))
  (Development|PrintString (Utilities|String|Append "Adding Weapon Small Preview Image Path" " : " (Utilities|String|Append (Utilities|String|Split (Utilities|GetObjectPathString (ItemPreviewImage|GetorLoadSynchronously (Item|GetWeaponDataSmallPreviewImage (Weapon|GetCurrentWeapon _asfort_player_pawn)))) "." "IgnoreCase" "FromEnd") ".TGA")))
  (Development|PrintString (Utilities|String|Append "Fort Player Pawn Weapon Small Preview Image" " : " (Utilities|GetObjectPathString (ItemPreviewImage|GetorLoadSynchronously (Item|GetWeaponDataSmallPreviewImage (Weapon|GetCurrentWeapon _asfort_player_pawn))))))
  (CallFunction|AddAssettobeExported (ItemPreviewImage|GetorLoadSynchronously (Item|GetWeaponDataSmallPreviewImage (Weapon|GetCurrentWeapon _asfort_player_pawn)))))

(fn InitializeVariables
  (Utilities|Array|Clear (Variables|Default|GetAssetsToExport)))

(fn TraceAndFilterDependencies
  (params (InAsset :type Object%20Reference) (PathKeyDirectory :type String) (PathKeyFileName :type String))
  ;[1 1 1] Clear Variables
  (Utilities|Array|Clear (LocalVariables|GetAssetDataList))
  (Utilities|Array|Clear (LocalVariables|GetFilteredDependencies))
  ;[1 1 1] Process Reference Dependencies
  (bind _outdependencies (AssetRegistry|GetDependencies (AssetRegistry|GetAssetRegistry) (Utilities|Struct|BreakAssetData (AssetData|CreateAssetData InAsset)) (Utilities|Struct|MakeAssetRegistryDependencyOptions true)))
  (for _array_element _outdependencies
    ;[1 0.2 0.2] TODO: Find a cleaner way to filter object path to remove this path hack. Ideally it would be good to look up the type and check to ensure it is an AnimSequence.
    (if (Utilities|String|Contains (Utilities|String|ToString(Name) _array_element) PathKeyFileName)
      (if (Utilities|String|Contains (Utilities|String|ToString(Name) _array_element) PathKeyDirectory)
        ;[1 1 1] Update Array Of Filtered Dependencies To Return
        (Utilities|Array|AppendArray (LocalVariables|GetFilteredDependencies) (Utilities|Array|MakeArray _array_element)))))
  (return (LocalVariables|GetFilteredDependencies)))

(fn ConstructionScript
  (Rendering|Components|TextRender|SetText (Variables|Default|GetTextRender) (Utilities|Text|ToText(String) (Variables|Default|GetTitle))))
