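# Editor Utility Widget for hard-surface QC in Fortnite that applies a chosen character and weapon item definition.
# On construct, Python reads the CosmeticDeveloperSettings character override and matches it to a character asset.
# Apply runs a GiveWeapon cheat, then Python writes the character override and calls the CosmeticPicker module.
# A Details View exposes the widget's SetupDetails properties; QC-actor, weapon-debug and referencer chains are disconnected from Apply.
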EditorUtilityWidgetBlueprint /EpicHardSurfaceQC/EUW_HardSurfaceQCTool
graphs [
 ApplyItemDefinitions {nodes [
   n1 = GetActorOfClass!(
    execute
    WorldContextObject:Object Reference
    ActorClass:Actor Class Reference=&/BP_HardSurfaceQCActor.BP_HardSurfaceQCActor_C
   ) -> (then, ReturnValue:BP Hard Surface QCActor Object Reference)
   > PrintString(InString=Append!(
     A=BP HardSurfaceQCTool
     B=" : "
     C:String=GetDisplayName(Object=@n1)
     D:String=" Current Character Item Definition Name"
     E:String=" : "
     F:String=Get CurrentCharacterItemDefinitionName!(self:BP Hard Surface QCActor Object Reference=@n1) -> (CurrentCharacterItemDefinitionName:String)
    ) -> (ReturnValue))
   > Set CurrentCharacterItemDefinitionName!(
    execute
    CurrentCharacterItemDefinitionName:String=GetDisplayName(Object=Get Character Item Definition() -> (Character Item Definition:Athena Character Item Definition Object Reference))
    self:BP Hard Surface QCActor Object Reference=@n1
   ) -> (then, Output_Get:String)
   > n2 = PrintString(InString=Append!(A=Number of BP_HardSurfaceQCTool actors, B=" : ", C) -> (ReturnValue))
   n3 = Append(
    A=Cheat GiveWeapon
    B=" "
    C=GetDisplayName(Object=Get Weapon Item Definition() -> (Weapon Item Definition:Fort Weapon Item Definition Object Reference))
   )
   PrintString(execute=[@n2.then, @n23.then], InString=@n3)
   > n4 = ExecuteConsoleCommand(Command=@n3)
   n5 = Append!(A=Character, B) -> (ReturnValue)
   n6 = Get Character Item Definition()
   Execute Python Script!(
    execute=[@n4.then, @n16.then]
    PythonScript="\r\nimport unreal\r\nimport sys\r\nimport csv\r\n\r\nsettingsClassString = \"/Script/FortniteGame.CosmeticDeveloperSettings\"\r\nsettingsSectionName = \"CosmeticLoadOutOverrides\"\r\nsettingsClass = unreal.load_class(None, settingsClassString)\r\n\r\n## Get the class of the Developer Settings\r\nsettingsCdo = unreal.get_default_object(settingsClass)\r\n\r\n## Get the default object so we can edit its properties\r\noverrides = settingsCdo.get_editor_property(settingsSectionName)\r\n\r\noverrides.set_editor_property(id_name, id)"
    idName:String=@n5
    id:Athena Character Item Definition Object Reference=@n6
   ) -> (then, ReturnValue) {inputs [idName,id]}
   > PrintString(InString=Append!(A=SetID:, B=@n5, C:String, D:String=GetDisplayName(Object=@n6)) -> (ReturnValue))
   > Execute Python Script(
    PythonScript="import TechArt.CosmeticPicker.CosmeticPickerModule as CosmeticPicker\r\nCosmeticPicker.setOverride(id_name, id)"
    idName=@n5
    id=@n6
   ) {inputs [idName,id]}
   n7 = Get Weapon Item Definition()
   PrintString(InString=Append!(A=Weapon Actor Path Name, B=" : ", C=Get Object Path String(Object=@n7)) -> (ReturnValue))
   > PrintString(InString=Append(A=Weapon Actor Object Name, B=" : ", C=GetObjectName(Object=@n7)))
   > PrintString(InString=Append(A=Weapon Actor Display Name, B=" : ", C=GetDisplayName(Object=@n7)))
   n8 = Get Weapon Item Definition()
   > PrintString(InString=Append(A=Path Name, B=" : ", C=Get Object Path String(Object=@n8)))
   n9 = GetWeaponActorClass!(self:Fort Weapon Item Definition Object Reference=@n8) -> (ReturnValue:Fort Weapon Class Reference)
   > PrintString(InString=Append(A=Actor Class, B=" : ", C=Get Class Display Name(Class=@n9)))
   n10 = GetWeaponMeshOverride!(self:Fort Weapon Item Definition Object Reference=@n8) -> (ReturnValue:Skeletal Mesh Object Reference)
   > PrintString(InString=Append(A=Mesh Override, B=" : ", C=GetDisplayName(Object=@n10)))
   n11 = GetDisplayName!(self:Item Definition Base Object Reference=@n8) -> (ReturnValue:Text)
   > PrintString(InString=Append(A=Dispaly Name, B=" : ", C="To String (Text)"(InText=@n11)))
   n12 = GetItemTypeName!(self:Fort Item Definition Object Reference=@n8, bUsePlural:bool=false) -> (ReturnValue:Text)
   > PrintString(InString=Append(A=Item Type Name, B=" : ", C="To String (Text)"(InText=@n12)))
   > PrintString(InString=Append(A=Persistent Name, B=" : ", C=GetPersistentName!(self:Fort Item Definition Object Reference=@n8) -> (ReturnValue:String)))
   n13 = GetQuantityDisplayName!(self:Fort Item Definition Object Reference=@n8, Quantity:int=0) -> (ReturnValue:Text)
   > PrintString(InString=Append(A=Quantity Dispaly Name, B=" : ", C="To String (Text)"(InText=@n13)))
   n14 = GetShortDisplayName!(self:Fort Weapon Item Definition Object Reference=@n8) -> (ReturnValue:Text)
   > PrintString(InString=Append(A=Short Dispaly Name, B=" : ", C="To String (Text)"(InText=@n14)))
   n15 = GetWeaponMeshOverride(self=@n8)
   > n16 = PrintString(InString=Append(A=Weapon Mesh Override, B=" : ", C=GetDisplayName!(Object:Object Reference=@n15) -> (ReturnValue:String)))
   n17 = FindPackageReferencersForAsset!(execute, AssetPath:String=Get Object Path String(Object=Get Weapon Item Definition()), bLoadAssetsToConfirm:bool=false) -> (then, ReturnValue:Array of Strings)
   > n18 = For Each Loop(Array=@n17)
   > PrintString(InString=Append(A=Referencer, B=" : ", C=@n18))
   n19 = Execute Python Script!(
    execute
    PythonScript="import unreal\r\nimport sys\r\nsettingsClassString = \"&CosmeticDeveloperSettings\"\r\nsettingsSectionName = \"CosmeticLoadOutOverrides\"\r\nsettingsClass = unreal.load_class(None, settingsClassString) ##get the class of the Developer Settings\r\nsettingsCdo = unreal.get_default_object(settingsClass) ##Get the default object so we can edit its properties\r\noverrides = settingsCdo.get_editor_property(settingsSectionName)\r\n\r\noverride_id_name = overrides.get_editor_property(id_name)\r\nid = None\r\nif override_id_name:\r\n id = unreal.FortItemDefinition.cast(override_id_name)\r\n"
    id_name:Wildcard
   ) -> (then, ReturnValue, id:Object Reference) {inputs [id_name], outputs [id]}
   > n20 = GetAllActorsOfClass!(
    execute
    WorldContextObject:Object Reference
    ActorClass:Actor Class Reference=&BP_HardSurfaceQCActor_C
   ) -> (then, OutActors:Array of BP Hard Surface QCActor Object References)
   > n21 = For Each Loop(Array:Array of BP Hard Surface QCActor Object References=@n20) -> (Array Element:BP Hard Surface QCActor Object Reference)
   > PrintString(InString=Append(A=BP_HardSurfaceQCTool Actor, B=" : ", C=GetDisplayName(Object=@n21)))
   > n22 = Set CurrentCharacterItemDefinitionName(CurrentCharacterItemDefinitionName=GetDisplayName(Object=@n19.id), self=@n21)
   > PrintString(InString=@n22)
   n23 = ApplyItemDefinitions() {is=title K2Node_FunctionEntry}
   n24 = Length(TargetArray:Array of Strings=@n17)
   PrintString(InString=Append(A=Referencer Count, B=" : ", C="To String (Integer)"(InInt=@n24)))
   Execute Python Script!(
    execute
    PythonScript="\r\nimport unreal\r\nimport sys\r\nimport csv\r\n\r\nsettingsClassString = \"&CosmeticDeveloperSettings\"\r\nsettingsSectionName = \"CosmeticLoadOutOverrides\"\r\nsettingsClass = unreal.load_class(None, settingsClassString)\r\n\r\n## Get the class of the Developer Settings\r\nsettingsCdo = unreal.get_default_object(settingsClass)\r\n\r\n## Get the default object so we can edit its properties\r\noverrides = settingsCdo.get_editor_property(settingsSectionName)\r\n\r\noverrides.set_editor_property(id_name, id)"
    idName:Wildcard
    id:Wildcard
   ) -> (then, ReturnValue) {inputs [idName,id]}
   Get CurrentCharacterItemDefinitionName(self=@n1)
  ]}
 EventGraph {nodes [
   "On Clicked (TestButton)"() {is=title K2Node_ComponentBoundEvent}
   > InitializeVariables()
   > ApplyItemDefinitions() {is=title K2Node_CallFunction}
   Event PreConstruct()
   > SetObject(self=Get SetupDetailsView() -> (SetupDetailsView:Details View Object Reference), NewObject=Self-Reference())
   Event Construct()
   > InitializeCharacterItemDefinition()
   Event Tick()
  ]}
 InitializeCharacterItemDefinition {nodes [
   n1 = Execute Python Script!(
    execute=@n13.then
    PythonScript="\r\nimport unreal\r\n##import sys\r\n##import csv\r\n\r\nsettingsClassString = \"&CosmeticDeveloperSettings\"\r\nsettingsClass = unreal.load_class(None, settingsClassString)\r\n\r\n## Get the class of the Developer Settings\r\nsettingsCdo = unreal.get_default_object(settingsClass)\r\n\r\n## Get the default object so we can edit its properties\r\nsettingsSectionName = \"CosmeticLoadOutOverrides\"\r\noverrides = settingsCdo.get_editor_property(settingsSectionName)\r\n\r\n## Inspect properties\r\n##print(\"overrides: {}\".format(overrides))\r\n\r\n\r\n## Output type: athena_character_item_definition\r\noutput_character = overrides.get_editor_property('character')\r\n##print(\"output_character: {}\".format(output_character))\r\noutput_character_get_full_name = ''\r\nif output_character:\r\n    ##members = [e for e in sorted(dir(output_character))]\r\n    ##print(\"members: {}\".format(members))\r\n\r\n    print(\"output_character.get_full_name(): {}\".format(output_character.get_full_name()))\r\n    output_character_get_full_name = str(output_character.get_full_name())\r\n\r\n\r\n##overrides.set_editor_property(id_name, id)\r\n"
   ) -> (then, ReturnValue, output_character_get_full_name:String) {outputs [output_character_get_full_name]}
   PrintString(InString=Append(A=Overrides - Character Item Definition, B=" : ", C=@n1.output_character_get_full_name))
   n2 = Get Character Item Definition()
   > PrintString(InString=Append(A=Current Character Item Definition Path, B=" : ", C=Get Object Path String(Object=@n2)))
   n3 = Append!(A=/, B=Split(SourceString=@n1.output_character_get_full_name, InStr=/, SearchDir=FromStart).RightS) -> (ReturnValue)
   > n4 = PrintString(InString=Append!(A=Character Item Definition Path, B=" : ", C:String=@n3) -> (ReturnValue))
   n5 = GetAssetsByPath!(
    execute=[@n4.then, @n16.then]
    self:Asset Registry Interface=GetAssetRegistry()
    PackagePath:Name=/Game/Athena/Items/Cosmetics/Characters
    bRecursive:bool=false
    bIncludeOnlyOnDiskAssets:bool=false
   ) -> (then, OutAssetData:Array of Asset Data Structures, ReturnValue:bool)
   > n6 = For Each Loop(Array:Array of Asset Data Structures=@n5) -> (Array Element:Asset Data Structure)
   > n7 = Branch(Condition="Equal Exactly (String)"(A=@n3, B=ToSoftObjectPath!(InAssetData:"Asset Data Structure (by ref)"=@n6) -> (ReturnValue_PathString:String)))
   n8 = GetAsset(InAssetData=@n6)
   n9 = Cast To AthenaCharacterItemDefinition!(execute=[@n12.then, @n7.then], Object=@n8) -> (then, CastFailed, AsAthena Character Item Definition:Athena Character Item Definition Object Reference) {is=title K2Node_DynamicCast}
   Set Character Item Definition(execute=[@n10.then, @n9.then], Character Item Definition:Athena Character Item Definition Object Reference=@n9) -> (Output_Get:Athena Character Item Definition Object Reference)
   PrintString(execute=@n9.CastFailed, InString=Cast Failed)
   n10 = PrintString(InString=Cast Succeeded)
   n11 = Length(TargetArray:Array of Asset Data Structures=@n5)
   PrintString(InString=Append(A=Assets from path, B=" : ", C="To String (Integer)"(InInt=@n11)))
   > n12 = PrintString(InString=Append(A=Get Asset by Path, B=" : ", C=GetDisplayName(Object=@n8)))
   n13 = InitializeCharacterItemDefinition() {is=title K2Node_FunctionEntry}
   n14 = IsEmpty(InString=@n1.output_character_get_full_name)
   n15 = NOT Boolean(A=@n14)
   n16 = Branch(execute=@n1.then, Condition=@n15)
  ]}
 InitializeVariables {nodes [
   InitializeVariables() {is=title K2Node_FunctionEntry}
   > Clear(TargetArray=Get ItemsToExport() -> (ItemsToExport:Array of Strings))
   > Set ShadersToTextures(ShadersToTextures:String) -> (Output_Get:String)
   > Set ShadersToTexturesFileName(ShadersToTexturesFileName:String) -> (Output_Get:String)
  ]}
]
variables [
 Character Item Definition {type=AthenaCharacterItemDefinition, default=/Game/Athena/Items/Cosmetics/Characters/CID_001_Athena_Commando_F_Default}
 CommandStr {type=string, default=Cheat GiveWeapon WID_Assault_PistolCaliber_AR_Athena_R_Ore_T03}
 ExportPath {type=DirectoryPath, default {path=D:/EngineTemplates}}
 ItemsToExport {type="array<string>", default []}
 MaterialInstanceWithTextures {type=MaterialInstance, default=null}
 ShadersToTextures {type=string, default=""}
 ShadersToTexturesFileName {type=string, default=""}
 Weapon Item Definition {type=FortWeaponItemDefinition, default=/Game/Athena/Items/Weapons/WID_Assault_Heavy_Athena_VR_Ore_T03}
]
widgets {
 info {rootWidgetClass=&ScrollBox, widgetCount=4}
 widgets [
  ScrollBox_19 {
   widgetClassPath=&ScrollBox
   widget=ScrollBox_19
   properties {
    widgetBarStyle {
     thickness=8
     draggedThumbImage {
      bIsDynamicallyLoaded=false
      drawAs=RoundedBox
      imageType=NoImage
      mirroring=NoMirror
      resourceName=None
      resourceObject=null
      tiling=NoTile
      imageSize=<0 0>
      margin=<ltrb 0 0 0 0>
      outlineSettings {
       bUseBrushTransparency=false
       roundingType=FixedRadius
       width=0
       color {colorUseRule=UseColor_Specified, specifiedColor=<rgba 0 0 0 0>}
       cornerRadii=<4 4 4 4>
      }
      tintColor {colorUseRule=UseColor_Specified, specifiedColor=<rgba 0.2159 0.2159 0.2159 1>}
      uVRegion {bIsValid=false, max=<0 0>, min=<0 0>}
     }
     horizontalBackgroundImage {
      bIsDynamicallyLoaded=false
      drawAs=NoDrawType
      imageType=NoImage
      mirroring=NoMirror
      resourceName=None
      resourceObject=null
      tiling=NoTile
      imageSize=<0 0>
      margin=<ltrb 0 0 0 0>
      outlineSettings {
       bUseBrushTransparency=false
       roundingType=HalfHeightRadius
       width=0
       color {colorUseRule=UseColor_Specified, specifiedColor=<rgba 0 0 0 0>}
       cornerRadii=<0 0 0 0>
      }
      tintColor {colorUseRule=UseColor_Specified, specifiedColor=<rgba 1 1 1 1>}
      uVRegion {bIsValid=false, max=<0 0>, min=<0 0>}
     }
     horizontalBottomSlotImage {
      bIsDynamicallyLoaded=false
      drawAs=NoDrawType
      imageType=NoImage
      mirroring=NoMirror
      resourceName=None
      resourceObject=null
      tiling=NoTile
      imageSize=<0 0>
      margin=<ltrb 0 0 0 0>
      outlineSettings {
       bUseBrushTransparency=false
       roundingType=HalfHeightRadius
       width=0
       color {colorUseRule=UseColor_Specified, specifiedColor=<rgba 0 0 0 0>}
       cornerRadii=<0 0 0 0>
      }
      tintColor {colorUseRule=UseColor_Specified, specifiedColor=<rgba 1 1 1 1>}
      uVRegion {bIsValid=false, max=<0 0>, min=<0 0>}
     }
     horizontalTopSlotImage {
      bIsDynamicallyLoaded=false
      drawAs=NoDrawType
      imageType=NoImage
      mirroring=NoMirror
      resourceName=None
      resourceObject=null
      tiling=NoTile
      imageSize=<0 0>
      margin=<ltrb 0 0 0 0>
      outlineSettings {
       bUseBrushTransparency=false
       roundingType=HalfHeightRadius
       width=0
       color {colorUseRule=UseColor_Specified, specifiedColor=<rgba 0 0 0 0>}
       cornerRadii=<0 0 0 0>
      }
      tintColor {colorUseRule=UseColor_Specified, specifiedColor=<rgba 1 1 1 1>}
      uVRegion {bIsValid=false, max=<0 0>, min=<0 0>}
     }
     hoveredThumbImage {
      bIsDynamicallyLoaded=false
      drawAs=RoundedBox
      imageType=NoImage
      mirroring=NoMirror
      resourceName=None
      resourceObject=null
      tiling=NoTile
      imageSize=<0 0>
      margin=<ltrb 0 0 0 0>
      outlineSettings {
       bUseBrushTransparency=false
       roundingType=FixedRadius
       width=0
       color {colorUseRule=UseColor_Specified, specifiedColor=<rgba 0 0 0 0>}
       cornerRadii=<4 4 4 4>
      }
      tintColor {colorUseRule=UseColor_Specified, specifiedColor=<rgba 0.2159 0.2159 0.2159 1>}
      uVRegion {bIsValid=false, max=<0 0>, min=<0 0>}
     }
     normalThumbImage {
      bIsDynamicallyLoaded=false
      drawAs=RoundedBox
      imageType=NoImage
      mirroring=NoMirror
      resourceName=None
      resourceObject=null
      tiling=NoTile
      imageSize=<0 0>
      margin=<ltrb 0 0 0 0>
      outlineSettings {
       bUseBrushTransparency=false
       roundingType=FixedRadius
       width=0
       color {colorUseRule=UseColor_Specified, specifiedColor=<rgba 0 0 0 0>}
       cornerRadii=<4 4 4 4>
      }
      tintColor {colorUseRule=UseColor_Specified, specifiedColor=<rgba 0.0953 0.0953 0.0953 1>}
      uVRegion {bIsValid=false, max=<0 0>, min=<0 0>}
     }
     verticalBackgroundImage {
      bIsDynamicallyLoaded=false
      drawAs=NoDrawType
      imageType=NoImage
      mirroring=NoMirror
      resourceName=None
      resourceObject=null
      tiling=NoTile
      imageSize=<0 0>
      margin=<ltrb 0 0 0 0>
      outlineSettings {
       bUseBrushTransparency=false
       roundingType=HalfHeightRadius
       width=0
       color {colorUseRule=UseColor_Specified, specifiedColor=<rgba 0 0 0 0>}
       cornerRadii=<0 0 0 0>
      }
      tintColor {colorUseRule=UseColor_Specified, specifiedColor=<rgba 1 1 1 1>}
      uVRegion {bIsValid=false, max=<0 0>, min=<0 0>}
     }
     verticalBottomSlotImage {
      bIsDynamicallyLoaded=false
      drawAs=NoDrawType
      imageType=NoImage
      mirroring=NoMirror
      resourceName=None
      resourceObject=null
      tiling=NoTile
      imageSize=<0 0>
      margin=<ltrb 0 0 0 0>
      outlineSettings {
       bUseBrushTransparency=false
       roundingType=HalfHeightRadius
       width=0
       color {colorUseRule=UseColor_Specified, specifiedColor=<rgba 0 0 0 0>}
       cornerRadii=<0 0 0 0>
      }
      tintColor {colorUseRule=UseColor_Specified, specifiedColor=<rgba 1 1 1 1>}
      uVRegion {bIsValid=false, max=<0 0>, min=<0 0>}
     }
     verticalTopSlotImage {
      bIsDynamicallyLoaded=false
      drawAs=NoDrawType
      imageType=NoImage
      mirroring=NoMirror
      resourceName=None
      resourceObject=null
      tiling=NoTile
      imageSize=<0 0>
      margin=<ltrb 0 0 0 0>
      outlineSettings {
       bUseBrushTransparency=false
       roundingType=HalfHeightRadius
       width=0
       color {colorUseRule=UseColor_Specified, specifiedColor=<rgba 0 0 0 0>}
       cornerRadii=<0 0 0 0>
      }
      tintColor {colorUseRule=UseColor_Specified, specifiedColor=<rgba 1 1 1 1>}
      uVRegion {bIsValid=false, max=<0 0>, min=<0 0>}
     }
    }
    widgetStyle {
     barThickness=8
     bottomShadowBrush {
      bIsDynamicallyLoaded=false
      drawAs=Box
      imageType=FullColor
      mirroring=NoMirror
      resourceName=../../../Engine/Content/Slate/Common/ScrollBoxShadowBottom.png
      resourceObject=null
      tiling=NoTile
      imageSize=<16 8>
      margin=<ltrb 0.5 0 0.5 1>
      outlineSettings {
       bUseBrushTransparency=false
       roundingType=HalfHeightRadius
       width=0
       color {colorUseRule=UseColor_Specified, specifiedColor=<rgba 0 0 0 0>}
       cornerRadii=<0 0 0 0>
      }
      tintColor {colorUseRule=UseColor_Specified, specifiedColor=<rgba 1 1 1 1>}
      uVRegion {bIsValid=false, max=<0 0>, min=<0 0>}
     }
     horizontalScrolledContentPadding=<ltrb 0 0 1 0>
     leftShadowBrush {
      bIsDynamicallyLoaded=false
      drawAs=Box
      imageType=FullColor
      mirroring=NoMirror
      resourceName=../../../Engine/Content/Slate/Common/ScrollBoxShadowLeft.png
      resourceObject=null
      tiling=NoTile
      imageSize=<8 16>
      margin=<ltrb 1 0.5 0 0.5>
      outlineSettings {
       bUseBrushTransparency=false
       roundingType=HalfHeightRadius
       width=0
       color {colorUseRule=UseColor_Specified, specifiedColor=<rgba 0 0 0 0>}
       cornerRadii=<0 0 0 0>
      }
      tintColor {colorUseRule=UseColor_Specified, specifiedColor=<rgba 1 1 1 1>}
      uVRegion {bIsValid=false, max=<0 0>, min=<0 0>}
     }
     rightShadowBrush {
      bIsDynamicallyLoaded=false
      drawAs=Box
      imageType=FullColor
      mirroring=NoMirror
      resourceName=../../../Engine/Content/Slate/Common/ScrollBoxShadowRight.png
      resourceObject=null
      tiling=NoTile
      imageSize=<8 16>
      margin=<ltrb 0 0.5 1 0.5>
      outlineSettings {
       bUseBrushTransparency=false
       roundingType=HalfHeightRadius
       width=0
       color {colorUseRule=UseColor_Specified, specifiedColor=<rgba 0 0 0 0>}
       cornerRadii=<0 0 0 0>
      }
      tintColor {colorUseRule=UseColor_Specified, specifiedColor=<rgba 1 1 1 1>}
      uVRegion {bIsValid=false, max=<0 0>, min=<0 0>}
     }
     topShadowBrush {
      bIsDynamicallyLoaded=false
      drawAs=Box
      imageType=FullColor
      mirroring=NoMirror
      resourceName=../../../Engine/Content/Slate/Common/ScrollBoxShadowTop.png
      resourceObject=null
      tiling=NoTile
      imageSize=<16 8>
      margin=<ltrb 0.5 1 0.5 0>
      outlineSettings {
       bUseBrushTransparency=false
       roundingType=HalfHeightRadius
       width=0
       color {colorUseRule=UseColor_Specified, specifiedColor=<rgba 0 0 0 0>}
       cornerRadii=<0 0 0 0>
      }
      tintColor {colorUseRule=UseColor_Specified, specifiedColor=<rgba 1 1 1 1>}
      uVRegion {bIsValid=false, max=<0 0>, min=<0 0>}
     }
     verticalScrolledContentPadding=<ltrb 0 0 0 1>
    }
   }
  }
  SetupDetailsView {
   widgetClassPath=&DetailsView
   parent=ScrollBox_19
   widget=SetupDetailsView
   properties {categoriesToShow [SetupDetails]}
   slot {class=ScrollBoxSlot}
  }
  TestButton {
   widgetClassPath=&Button
   parent=ScrollBox_19
   widget=TestButton
   properties {widgetStyle {
     clickedSlateSound {resourceObject=null}
     disabled {
      bIsDynamicallyLoaded=false
      drawAs=RoundedBox
      imageType=NoImage
      mirroring=NoMirror
      resourceName=None
      resourceObject=null
      tiling=NoTile
      imageSize=<0 0>
      margin=<ltrb 0 0 0 0>
      outlineSettings {
       bUseBrushTransparency=false
       roundingType=FixedRadius
       width=1
       color {colorUseRule=UseColor_ColorTable, specifiedColor=<rgba 0 0 0 0>}
       cornerRadii=<4 4 4 4>
      }
      tintColor {colorUseRule=UseColor_Specified, specifiedColor=<rgba 0.0395 0.0395 0.0395 1>}
      uVRegion {bIsValid=false, max=<0 0>, min=<0 0>}
     }
     disabledForeground {colorUseRule=UseColor_ColorTable, specifiedColor=<rgba 0 0 0 0>}
     hovered {
      bIsDynamicallyLoaded=false
      drawAs=RoundedBox
      imageType=NoImage
      mirroring=NoMirror
      resourceName=None
      resourceObject=null
      tiling=NoTile
      imageSize=<32 32>
      margin=<ltrb 0 0 0 0>
      outlineSettings {
       bUseBrushTransparency=false
       roundingType=FixedRadius
       width=1
       color {colorUseRule=UseColor_ColorTable, specifiedColor=<rgba 0 0 0 0>}
       cornerRadii=<4 4 4 4>
      }
      tintColor {colorUseRule=UseColor_Specified, specifiedColor=<rgba 0.0953 0.0953 0.0953 1>}
      uVRegion {bIsValid=false, max=<0 0>, min=<0 0>}
     }
     hoveredForeground {colorUseRule=UseColor_ColorTable, specifiedColor=<rgba 0 0 0 0>}
     hoveredSlateSound {resourceObject=null}
     normal {
      bIsDynamicallyLoaded=false
      drawAs=RoundedBox
      imageType=NoImage
      mirroring=NoMirror
      resourceName=None
      resourceObject=null
      tiling=NoTile
      imageSize=<32 32>
      margin=<ltrb 0 0 0 0>
      outlineSettings {
       bUseBrushTransparency=false
       roundingType=FixedRadius
       width=1
       color {colorUseRule=UseColor_ColorTable, specifiedColor=<rgba 0 0 0 0>}
       cornerRadii=<4 4 4 4>
      }
      tintColor {colorUseRule=UseColor_Specified, specifiedColor=<rgba 0.0395 0.0395 0.0395 1>}
      uVRegion {bIsValid=false, max=<0 0>, min=<0 0>}
     }
     normalForeground {colorUseRule=UseColor_ColorTable, specifiedColor=<rgba 0 0 0 0>}
     normalPadding=<ltrb 12 1.5 12 1.5>
     pressed {
      bIsDynamicallyLoaded=false
      drawAs=RoundedBox
      imageType=NoImage
      mirroring=NoMirror
      resourceName=None
      resourceObject=null
      tiling=NoTile
      imageSize=<32 32>
      margin=<ltrb 0 0 0 0>
      outlineSettings {
       bUseBrushTransparency=false
       roundingType=FixedRadius
       width=1
       color {colorUseRule=UseColor_ColorTable, specifiedColor=<rgba 0 0 0 0>}
       cornerRadii=<4 4 4 4>
      }
      tintColor {colorUseRule=UseColor_Specified, specifiedColor=<rgba 0.0284 0.0284 0.0284 1>}
      uVRegion {bIsValid=false, max=<0 0>, min=<0 0>}
     }
     pressedForeground {colorUseRule=UseColor_ColorTable, specifiedColor=<rgba 0 0 0 0>}
     pressedPadding=<ltrb 12 2.5 12 0.5>
     pressedSlateSound {resourceObject=null}
    }}
   slot {class=ScrollBoxSlot}
  }
  CamInfo_1 {
   widgetClassPath=&TextBlock
   bIsVariable=true
   parent=TestButton
   widget=CamInfo_1
   properties {
    text=Apply
    colorAndOpacity {colorUseRule=UseColor_Specified, specifiedColor=<rgba 1 1 1 0.8>}
    font {
     bForceMonospaced=false
     bMaterialIsStencil=false
     fontMaterial=null
     fontObject=&Roboto
     letterSpacing=0
     monospacedWidth=1
     size=12
     skewAmount=0
     typefaceFontName=Bold
     outlineSettings {
      bApplyOutlineToDropShadows=false
      bMiteredCorners=false
      bSeparateFillAlpha=false
      outlineMaterial=null
      outlineSize=2
      outlineColor=<rgba 0 0 0 0.8>
     }
    }
    margin=<ltrb 8 0 0 0>
   }
   slot {class=ButtonSlot, properties {verticalAlignment=VAlign_Top, padding=<ltrb 0 0 0 0>}}
  }
 ]
}
